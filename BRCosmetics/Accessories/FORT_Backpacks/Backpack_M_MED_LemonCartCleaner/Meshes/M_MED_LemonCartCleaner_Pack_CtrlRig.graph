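# Control Rig for the LemonCartCleaner backpack that repositions the pack according to the wearer's body pose.
# A fixed offset is applied to the pack bone, then isSkydive selects between two sets of weighted root-bone adjustments.
# Weights come from remapped pelvis-to-spine relative yaw and roll plus pelvis pitch, some multiplied together.
# The skydive set can also shrink the pack; the bike and rider bones are never driven.
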ControlRigBlueprint /BRCosmetics/Accessories/FORT_Backpacks/Backpack_M_MED_LemonCartCleaner/Meshes/M_MED_LemonCartCleaner_Pack_CtrlRig
graphs [
 RigVMFunctionLibrary {links [], nodes []}
 RigVMModel {nodes [
   RigUnit_BeginExecution = &RigUnit_BeginExecution()
   n1 = &RigUnit_ModifyTransforms(
    ItemToModify=((Item=(Type=Bone,Name="pack"),Transform=(Rotation=(X=0.0646,Y=-0.0843,Z=0.0487,W=0.9932),Translation=(X=0,Y=3.5,Z=0),Scale3D=(X=1,Y=1,Z=1))))
    Weight=1
    WeightMinimum=0
    ExecutePin=@RigUnit_BeginExecution.ExecutePin
   )
   RigVMFunction_ControlFlowBranch = &RigVMFunction_ControlFlowBranch(ExecuteContext=@n1.ExecutePin, Condition=RigVMVariableNode() {variable=isSkydive}.Value)
   n2 = &RigVMFunction_ControlFlowBranch(Condition=False, ExecuteContext=@RigVMFunction_ControlFlowBranch.False)
   RigVMFunction_Sequence = &RigVMFunction_Sequence(ExecuteContext=@n2.Completed)
   n3 = RigVMVariableNode() {variable=TPelvis}
   n4 = RigVMVariableNode() {variable=TSpine03}
   DISPATCH_RigVMDispatch_BreakStruct = RigVMDispatchNode(Struct=/Script/RigVM.RigVMFunction_MathQuaternionToRotator(Value=&RigVMFunction_MathTransformMakeRelative(Parent=@n3.Value, Global=@n4.Value).Local.Rotation) {is=struct RigVMUnitNode, template="Cast::Execute(in Value,out Result)"}.Result) {template="DISPATCH_RigVMDispatch_BreakStruct(in Struct,out Elements)"}
   &RigUnit_ModifyTransforms(
    ItemToModify=((Item=(Type=Bone,Name="root"),Transform=(Rotation=(X=-0.1305,Y=0,Z=0,W=0.9914),Translation=(X=0,Y=0,Z=5),Scale3D=(X=1,Y=1,Z=1))))
    WeightMinimum=0
    Weight=&RigVMFunction_MathDoubleRemap(SourceMaximum=-30, SourceMinimum=0, TargetMaximum=1, TargetMinimum=0, Value=@DISPATCH_RigVMDispatch_BreakStruct.Elements.Yaw).Result
    ExecutePin=@RigVMFunction_Sequence.A
   )
   > &RigUnit_ModifyTransforms(
    ItemToModify=((Item=(Type=Bone,Name="root"),Transform=(Rotation=(X=-0.0872,Y=0,Z=0,W=0.9962),Translation=(X=0,Y=0,Z=0),Scale3D=(X=1,Y=1,Z=1))))
    WeightMinimum=0
    Weight=&RigVMFunction_MathDoubleRemap(SourceMaximum=45, SourceMinimum=0, TargetMaximum=1, TargetMinimum=0, Value=@DISPATCH_RigVMDispatch_BreakStruct.Elements.Yaw).Result
   )
   > &RigUnit_ModifyTransforms(
    ItemToModify=((Item=(Type=Bone,Name="root"),Transform=(Rotation=(X=0,Y=0,Z=0,W=1),Translation=(X=0,Y=0,Z=0),Scale3D=(X=1,Y=1,Z=1))))
    WeightMinimum=0
    Weight=&RigVMFunction_MathDoubleRemap(SourceMaximum=-60, SourceMinimum=-30, TargetMaximum=1, TargetMinimum=0, Value=@DISPATCH_RigVMDispatch_BreakStruct.Elements.Yaw).Result
   )
   Cast = &RigVMFunction_MathQuaternionToRotator(Value=@n3.Value.Rotation) {template="Cast::Execute(in Value,out Result)"}
   Remap = &RigVMFunction_MathDoubleRemap(
    SourceMaximum=70
    SourceMinimum=50
    TargetMaximum=1
    TargetMinimum=0
    Value=RigVMDispatchNode(Struct=@Cast.Result) {template="DISPATCH_RigVMDispatch_BreakStruct(in Struct,out Elements)"}.Elements.Pitch
   )
   Multiply = &RigVMFunction_MathDoubleMul(
    A=@Remap.Result
    B=&RigVMFunction_MathDoubleRemap(SourceMaximum=-20, SourceMinimum=-5, TargetMaximum=1, TargetMinimum=0, Value=@DISPATCH_RigVMDispatch_BreakStruct.Elements.Roll).Result
   )
   > &RigUnit_ModifyTransforms(
    ItemToModify=((Item=(Type=Bone,Name="root"),Transform=(Rotation=(X=-0.1677,Y=-0.0449,Z=0.2549,W=0.9513),Translation=(X=0,Y=2,Z=0),Scale3D=(X=1,Y=1,Z=1))))
    WeightMinimum=0
    Weight=@Multiply.Result
   )
   n5 = &RigVMFunction_MathQuaternionToRotator(Value=&RigVMFunction_MathTransformMakeRelative(Global=RigVMVariableNode() {variable=TSpine_05}.Value, Parent=RigVMVariableNode() {variable=TPelvis}.Value).Local.Rotation) {template="Cast::Execute(in Value,out Result)"}
   &RigUnit_ModifyTransforms(
    ItemToModify=((Item=(Type=Bone,Name="root"),Transform=(Rotation=(X=-0.1736,Y=0,Z=0,W=0.9848),Translation=(X=0,Y=0,Z=0),Scale3D=(X=1,Y=1,Z=1))))
    WeightMinimum=0
    Weight=&RigVMFunction_MathDoubleRemap(
     SourceMaximum=-20
     SourceMinimum=-5
     TargetMaximum=1
     TargetMinimum=0
     Value=RigVMDispatchNode(Struct=@n5.Result) {template="DISPATCH_RigVMDispatch_BreakStruct(in Struct,out Elements)"}.Elements.Yaw
    ).Result
    ExecutePin=@RigVMFunction_Sequence.B
   )
   n6 = &RigVMFunction_Sequence(ExecuteContext=@RigVMFunction_ControlFlowBranch.True)
   n7 = RigVMVariableNode() {variable=TPelvis}
   n8 = RigVMVariableNode() {variable=TSpine03}
   n9 = RigVMDispatchNode(Struct=&RigVMFunction_MathQuaternionToRotator(Value=&RigVMFunction_MathTransformMakeRelative(Parent=@n7.Value, Global=@n8.Value).Local.Rotation) {template="Cast::Execute(in Value,out Result)"}.Result) {template="DISPATCH_RigVMDispatch_BreakStruct(in Struct,out Elements)"}
   &RigUnit_ModifyTransforms(
    ItemToModify=((Item=(Type=Bone,Name="root"),Transform=(Rotation=(X=-0.0436,Y=0,Z=0,W=0.999),Translation=(X=0,Y=0,Z=0),Scale3D=(X=1,Y=1,Z=1))))
    WeightMinimum=0
    Weight=&RigVMFunction_MathDoubleRemap(SourceMaximum=-30, SourceMinimum=-10, TargetMaximum=1, TargetMinimum=0, Value=@n9.Elements.Yaw).Result
    ExecutePin=@n6.A
   )
   > &RigUnit_ModifyTransforms(
    ItemToModify=((Item=(Type=Bone,Name="root"),Transform=(Rotation=(X=-0.0872,Y=0,Z=0,W=0.9962),Translation=(X=0,Y=0,Z=0),Scale3D=(X=1,Y=1,Z=1))))
    WeightMinimum=0
    Weight=&RigVMFunction_MathDoubleRemap(SourceMaximum=45, SourceMinimum=0, TargetMaximum=1, TargetMinimum=0, Value=@n9.Elements.Yaw).Result
   )
   n10 = RigVMVariableNode() {variable=TSpine_05}
   n11 = RigVMVariableNode() {variable=TPelvis}
   n12 = RigVMDispatchNode(Struct=&RigVMFunction_MathQuaternionToRotator(Value=&RigVMFunction_MathTransformMakeRelative(Global=@n10.Value, Parent=@n11.Value).Local.Rotation) {template="Cast::Execute(in Value,out Result)"}.Result) {template="DISPATCH_RigVMDispatch_BreakStruct(in Struct,out Elements)"}
   > &RigUnit_ModifyTransforms(
    ItemToModify=((Item=(Type=Bone,Name="root"),Transform=(Rotation=(X=-0.1736,Y=0,Z=0,W=0.9848),Translation=(X=0,Y=-3,Z=0),Scale3D=(X=1,Y=1,Z=1))))
    WeightMinimum=0
    Weight=&RigVMFunction_MathDoubleMul(
     A=&RigVMFunction_MathDoubleRemap(SourceMaximum=12, SourceMinimum=0, TargetMaximum=1, TargetMinimum=0, Value=@n9.Elements.Yaw).Result
     B=&RigVMFunction_MathDoubleRemap(SourceMaximum=-25, SourceMinimum=0, TargetMaximum=1, TargetMinimum=0, Value=@n12.Elements.Yaw).Result
    ).Result
   )
   > &RigUnit_ModifyTransforms(
    ItemToModify=((Item=(Type=Bone,Name="root"),Transform=(Rotation=(X=-0.1677,Y=-0.0449,Z=0.2549,W=0.9513),Translation=(X=0,Y=2,Z=0),Scale3D=(X=1,Y=1,Z=1))))
    WeightMinimum=0
    Weight=&RigVMFunction_MathDoubleRemap(SourceMaximum=-20, SourceMinimum=-5, TargetMaximum=1, TargetMinimum=0, Value=@n9.Elements.Roll).Result
   )
   &RigUnit_ModifyTransforms(
    ItemToModify=((Item=(Type=Bone,Name="root"),Transform=(Rotation=(X=0,Y=0,Z=0.1305,W=0.9914),Translation=(X=0,Y=-3,Z=0),Scale3D=(X=1,Y=1,Z=1))))
    WeightMinimum=0
    Weight=&RigVMFunction_MathDoubleRemap(SourceMaximum=-20, SourceMinimum=-5, TargetMaximum=1, TargetMinimum=0, Value=@n12.Elements.Yaw).Result
    ExecutePin=@n6.B
   )
   > &RigUnit_ModifyTransforms(
    ItemToModify=((Item=(Type=Bone,Name="root"),Transform=(Rotation=(X=-0.2164,Y=0,Z=0,W=0.9763),Translation=(X=0,Y=-12,Z=0),Scale3D=(X=0.85,Y=0.85,Z=0.85))))
    WeightMinimum=0
    Weight=&RigVMFunction_MathDoubleRemap(SourceMaximum=-70, SourceMinimum=-40, TargetMaximum=1, TargetMinimum=0, Value=@n12.Elements.Yaw).Result
   )
  ]}
]
hierarchy [
 root {loc=<0 0 0>, rot=<0 0 0 1>}
 pack {parent=root, loc=<-0.963 -11.169 -27.454>, rot=<0 0 0 1>}
 bp_bike_root {parent=pack, loc=<-1.0625 -0.0137 0.0123>, rot=<0 0 -0.7115 -0.7027>}
 bp_bike_wheelie {parent=bp_bike_root, loc=<-0.0013 -7.9274 -0.0029>, rot=<0 0 0 1>}
 bp_bike_main {parent=bp_bike_wheelie, loc=<0.002 7.7852 7.8047>, rot=<0 0 0 1>}
 bp_wheel_back {parent=bp_bike_main, loc=<-0.0002 -8.4753 -3.6207>, rot=<0 0 0 1>}
 bp_bikeHandle {parent=bp_bike_main, loc=<-0.0003 6.1441 3.024>, rot=<0.4325 -0.4318 -0.5601 -0.5592>}
 bp_wheel_front {parent=bp_bikeHandle, loc=<6.8789 -0.0028 0.0418>, rot=<-0.4325 0.4318 0.5601 -0.5592>}
 bp_bikeStar {parent=bp_bike_main, loc=<-0.0002 -0.5238 -3.6751>, rot=<0 0 0 1>}
 bp_pedal_r {parent=bp_bikeStar, loc=<-2.7758 -1.7535 -0.1771>, rot=<-0.0285 0 0 -0.9996>}
 bp_pedal_l {parent=bp_bikeStar, loc=<2.7757 1.7468 0.1949>, rot=<0 0 0 1>}
 bp_pelvis {parent=bp_bike_wheelie, loc=<0.0018 5.205 12.9438>, rot=<0.3885 0.3884 0.5909 -0.5908>}
 bp_spine {parent=bp_pelvis, loc=<1.633 0 0.063>, rot=<-0.0001 0.1141 0 -0.9935>}
 bp_spine01 {parent=bp_spine, loc=<2.5699 0 0>, rot=<0 0.0742 0 -0.9972>}
 bp_spine02 {parent=bp_spine01, loc=<2.5875 0 0>, rot=<0 0.0573 0 -0.9984>}
 bp_neck01 {parent=bp_spine02, loc=<0.943 0 0.0139>, rot=<0 -0.1255 0 -0.9921>}
 bp_neck02 {parent=bp_neck01, loc=<3.1553 0 0>, rot=<0 0.1092 0 -0.994>}
 bp_head {parent=bp_neck02, loc=<1.7214 0 0>, rot=<0 -0.0249 0 -0.9997>}
 bp_snout01 {parent=bp_head, loc=<2.8881 0 3.7785>, rot=<0.6088 0 0.7933 0>}
 bp_snout02 {parent=bp_snout01, loc=<3.9494 0 0>, rot=<0 0.0523 0 -0.9986>}
 bp_snout03 {parent=bp_snout02, loc=<3.3279 0 0>, rot=<0 -0.0164 0 -0.9999>}
 bp_snout04 {parent=bp_snout03, loc=<2.4122 0 2.9313>, rot=<0 0.3284 0 -0.9445>}
 bp_glasses {parent=bp_head, loc=<6.8327 0 -0.5866>, rot=<0.7076 0 0.7066 0>}
 bp_brow_l {parent=bp_head, loc=<8.6669 1.8995 3.0873>, rot=<0.8932 0.0371 0.4419 0.0747>}
 bp_brow_r {parent=bp_head, loc=<8.6669 -1.8995 3.0873>, rot=<-0.4419 -0.0747 0.8932 0.0371>}
 bp_hair_main {parent=bp_head, loc=<9.7425 0.0652 -3.3664>, rot=<-0.5004 -0.5004 -0.4996 -0.4996>}
 bp_hair05 {parent=bp_hair_main, loc=<2.271 -1.0121 0.7196>, rot=<-0.0166 0.1742 0.327 -0.9287>}
 bp_hair04 {parent=bp_hair_main, loc=<0.1103 -2.4332 0.3621>, rot=<0 0 0.6918 -0.7221>}
 bp_hair08 {parent=bp_hair_main, loc=<1.9368 1.1082 2.2831>, rot=<-0.2746 0.1366 -0.4419 -0.843>}
 bp_hair02 {parent=bp_hair_main, loc=<-2.2754 -1.2131 -2.0217>, rot=<-0.1407 -0.082 0.8893 -0.4274>}
 bp_hair03 {parent=bp_hair_main, loc=<2.1127 -0.29 -1.3327>, rot=<-0.1244 -0.2481 0.132 -0.9516>}
 bp_hair07 {parent=bp_hair_main, loc=<-1.5397 1.839 0.1156>, rot=<-0.1207 0.2021 -0.7875 -0.5696>}
 bp_hair06 {parent=bp_hair_main, loc=<-1.5668 -1.046 1.4364>, rot=<0.197 -0.0187 0.9534 -0.2277>}
 bp_hair01 {parent=bp_hair_main, loc=<-2.5411 0.9378 -1.6433>, rot=<-0.1864 -0.0386 0.9799 -0.0595>}
 bp_hoodieCape01 {parent=bp_spine02, loc=<2.0038 -0.137 -2.5915>, rot=<-0.1634 0.688 0.688 0.1634>}
 bp_hoodieCape02 {parent=bp_hoodieCape01, loc=<1.5491 0.0666 0>, rot=<0 0 -0.2543 -0.9671>}
 bp_clavicle_r {parent=bp_spine02, loc=<1.205 -1.691 -0.1966>, rot=<0.5857 0.2572 -0.4351 -0.6336>}
 bp_upperarm_r {parent=bp_clavicle_r, loc=<-1.7573 -0.0001 -0.0002>, rot=<0.0615 -0.2181 -0.2123 -0.9506>}
 bp_lowerarm_r {parent=bp_upperarm_r, loc=<-4.7337 0 0>, rot=<0.5326 0.3618 -0.3198 -0.6951>}
 bp_hand_r {parent=bp_lowerarm_r, loc=<-3.8929 0 0.0001>, rot=<-0.2511 0.097 0.0117 -0.963>}
 bp_clavicle_l {parent=bp_spine02, loc=<1.205 1.691 -0.1966>, rot=<-0.4351 -0.6336 -0.5857 -0.2572>}
 bp_upperarm_l {parent=bp_clavicle_l, loc=<1.7573 0.0001 0.0001>, rot=<0.0615 -0.218 -0.2123 -0.9506>}
 bp_lowerarm_l {parent=bp_upperarm_l, loc=<4.7337 0 0>, rot=<0.5326 0.3618 -0.3198 -0.6951>}
 bp_hand_l {parent=bp_lowerarm_l, loc=<3.8929 0 0>, rot=<-0.2511 0.097 0.0117 -0.963>}
 joint38 {parent=bp_spine01, loc=<0.3862 0.046 3.3124>, rot=<0.7471 0 0.6647 0>}
 bp_tail01 {parent=bp_pelvis, loc=<-0.0111 0.0004 -2.341>, rot=<-0.0001 -0.7099 -0.0001 -0.7043>}
 bp_tail02 {parent=bp_tail01, loc=<2.4119 0 0>, rot=<0 0.2993 0 -0.9542>}
 bp_tail03 {parent=bp_tail02, loc=<2.0469 0 0>, rot=<0 0.3688 0 -0.9295>}
 bp_tail04 {parent=bp_tail03, loc=<1.5154 0 0>, rot=<0.0071 0.196 0.0032 -0.9806>}
 bp_thigh_l {parent=bp_pelvis, loc=<-1.4158 2.4178 0.9546>, rot=<0.3936 0.0457 0.9169 0.0483>}
 bp_calf_l {parent=bp_thigh_l, loc=<3.3211 0 0>, rot=<0.7663 0.0084 -0.6422 0.0146>}
 bp_foot_l {parent=bp_calf_l, loc=<4.6846 0 0>, rot=<0.7579 0.0025 -0.6523 -0.0042>}
 bp_toe_l {parent=bp_foot_l, loc=<3.0717 0 0>, rot=<0 0.123 0 -0.9924>}
 bp_thigh_r {parent=bp_pelvis, loc=<-1.4203 -2.397 0.3481>, rot=<0.1537 0.6458 0.7065 -0.2451>}
 bp_calf_r {parent=bp_thigh_r, loc=<3.1717 0 0>, rot=<-0.0064 -0.0152 0.7194 -0.6944>}
 bp_foot_r {parent=bp_calf_r, loc=<4.9907 0 0>, rot=<-0.0365 0.0025 -0.6085 -0.7927>}
 bp_toe_r {parent=bp_foot_r, loc=<3.05 0 0>, rot=<0 0.0059 -0.0676 -0.9977>}
 NewBone {loc=<0 0 0>, rot=<0 0 0 1>}
]
variables [
 TActor {type=FTransform, default=(Rotation=(X=0,Y=0,Z=0,W=1),Translation=(X=0,Y=0,Z=0),Scale3D=(X=1,Y=1,Z=1))}
 TPelvis {type=FTransform, default=(Rotation=(X=0,Y=0,Z=0,W=1),Translation=(X=0,Y=0,Z=0),Scale3D=(X=1,Y=1,Z=1))}
 TSpine03 {type=FTransform, default=(Rotation=(X=0,Y=0,Z=0,W=1),Translation=(X=0,Y=0,Z=0),Scale3D=(X=1,Y=1,Z=1))}
 TSpine_02 {type=FTransform, default=(Rotation=(X=0,Y=0,Z=0,W=1),Translation=(X=0,Y=0,Z=0),Scale3D=(X=1,Y=1,Z=1))}
 TSpine_05 {type=FTransform, default=(Rotation=(X=0,Y=0,Z=0,W=1),Translation=(X=0,Y=0,Z=0),Scale3D=(X=1,Y=1,Z=1))}
 isSkydive {type=bool, default=False}
]
