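# Animation Blueprint for the LemonCartCleaner back bling, playing an idle loop and a one-shot reaction when an enemy player is eliminated.
# FrontWheelSpin and RearWheelSpin curves are integrated over delta time into additive wheel-bone rotations.
# A Control Rig then receives the character's pelvis, spine and component world transforms plus skydive state, before a fixed bike-root offset.
# Gravity override is computed each update but unused in the AnimGraph.
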AnimBlueprint /BRCosmetics/Accessories/FORT_Backpacks/Backpack_M_MED_LemonCartCleaner/Meshes/M_MED_LemonCartCleaner_Pack_AnimBP
target_skeleton=&/M_MED_LemonCartCleaner_Pack_Skeleton
graphs [
 AnimGraph {nodes [
   AnimGraphNode_StateMachine_1 = Animation_Idle() {is=title AnimGraphNode_StateMachine}
   n1 = "Transform (Modify) Bone - Bone: bp_wheel_front"(
    ComponentPose=Local To Component(LocalPose=@AnimGraphNode_StateMachine_1)
    Alpha=1
    Translation="0,0,0"
    Rotation=Get WheelSpin_Front() -> (WheelSpin_Front:Rotator)
    Scale="1,1,1"
   ) {rotationMode=BMM_Additive, rotationSpace=BCS_BoneSpace, alphaBoolBlend {blendOption=Linear}, boneToModify {boneName=bp_wheel_front}}
   n2 = "Transform (Modify) Bone - Bone: bp_wheel_back"(ComponentPose=@n1, Alpha=1, Translation="0,0,0", Rotation=Get WheelSpin_Rear() -> (WheelSpin_Rear:Rotator), Scale="1,1,1") {rotationMode=BMM_Additive, rotationSpace=BCS_BoneSpace, alphaBoolBlend {blendOption=Linear}, boneToModify {boneName=bp_wheel_back}}
   n3 = Control Rig!(
    Source=Component To Local(ComponentPose=@n2)
    Alpha
    TSpine_02:Transform=Get TSpine_02() -> (TSpine_02:Transform)
    TSpine_05:Transform=Get TSpine_05() -> (TSpine_05:Transform)
    TPelvis:Transform=Get TPelvis() -> (TPelvis:Transform)
    isSkydive:bool=Get bIsSkyDiving()
    TActor:Transform=Get TActor() -> (TActor:Transform)
    TSpine03:Transform=Get TSpine03() -> (TSpine03:Transform)
   ) -> (Pose) {
    alphaBoolBlend {blendOption=Linear}
    controlRigAssetReference {blueprintRigClass=&/M_MED_LemonCartCleaner_Pack_CtrlRig.M_MED_LemonCartCleaner_Pack_CtrlRig_C}
   }
   n4 = "Transform (Modify) Bone - Bone: bp_bike_root"(ComponentPose=Local To Component(LocalPose=@n3), Alpha=1, Translation="0,0,10", Rotation="-10,0,0", Scale="1,1,1") {
    rotationMode=BMM_Additive
    translationMode=BMM_Additive
    alphaBoolBlend {blendOption=Linear}
    boneToModify {boneName=bp_bike_root}
    rotation {pitch=-10}
    translation {z=10}
   }
   Output Pose(Result=Component To Local(ComponentPose=@n4)) {layerGroup=None}
  ]}
 Animation_Idle {
  entry=Idle
  path=AnimGraph.AnimGraphNode_StateMachine_1.Animation_Idle
  ids=id
  nodes [
   AnimStateNode_0 = Idle()
   AnimStateNode_1 = Elim() {is=title AnimStateNode}
   AnimStateTransitionNode_0 = AnimStateTransitionNode() {crossfadeDuration=0.4}
   AnimStateTransitionNode_1 = AnimStateTransitionNode() {bAutomaticRuleBasedOnSequencePlayerInState=true}
  ]
  links [
   {
    dst=Idle
    rule=AnimGraph.AnimGraphNode_StateMachine_1.Animation_Idle.AnimStateTransitionNode_1.Transition
    src=Elim
    via=AnimStateTransitionNode_1
   }
   {
    dst=Elim
    rule=AnimGraph.AnimGraphNode_StateMachine_1.Animation_Idle.AnimStateTransitionNode_0.Transition
    src=Idle
    via=AnimStateTransitionNode_0
   }
  ]
 }
 Idle {
  path=AnimGraph.AnimGraphNode_StateMachine_1.Animation_Idle.AnimStateNode_0.Idle
  nodes [Output Animation Pose(Result=Sequence Player 'LemonCartCleaner_idle_nowheels_PACK'() {sequence=/BRCosmetics/Accessories/FORT_Backpacks/Backpack_M_MED_LemonCartCleaner/Animation/LemonCartCleaner_idle_nowheels_PACK})]
 }
 Elim {
  path=AnimGraph.AnimGraphNode_StateMachine_1.Animation_Idle.AnimStateNode_1.Elim
  nodes [
   n1 = Sequence Player 'LemonCartCleaner_reactive_nowheels_PACK'() {
    bLoopAnimation=false
    sequence=/BRCosmetics/Accessories/FORT_Backpacks/Backpack_M_MED_LemonCartCleaner/Animation/LemonCartCleaner_reactive_nowheels_PACK
   }
   Output Animation Pose(Result=@n1)
  ]
 }
 Transition {
  path=AnimGraph.AnimGraphNode_StateMachine_1.Animation_Idle.AnimStateTransitionNode_0.Transition
  nodes [Result(bCanEnterTransition=Get Active Kill() -> (Active Kill:bool))]
 }
 Transition {path=AnimGraph.AnimGraphNode_StateMachine_1.Animation_Idle.AnimStateTransitionNode_1.Transition, nodes [Result()]}
 EventGraph {nodes [
   n1 = Event BlueprintUpdateAnimation()
   > Set Delta Time X(Delta Time X=@n1.DeltaTimeX)
   n2 = Get Custom Base Mesh() -> (Custom Base Mesh:Skeletal Mesh Component Object Reference)
   > Set TSpine_05(TSpine_05:Transform=GetSocketTransform(self=@n2, InSocketName=spine_05, TransformSpace=RTS_World)) -> (Output_Get:Transform)
   > Set TSpine_02(TSpine_02:Transform=GetSocketTransform(self=@n2, InSocketName=spine_02, TransformSpace=RTS_World)) -> (Output_Get:Transform)
   > Set TPelvis(TPelvis:Transform=GetSocketTransform(self=@n2, InSocketName=pelvis, TransformSpace=RTS_World)) -> (Output_Get:Transform)
   > Set TActor(TActor:Transform=Get World Transform!(self:Scene Component Object Reference=@n2) -> (ReturnValue:Transform) {is=title K2Node_CallFunction}) -> (Output_Get:Transform)
   > Set TSpine03(TSpine03:Transform=GetSocketTransform(self=@n2, InSocketName=spine_03, TransformSpace=RTS_World)) -> (Output_Get:Transform)
   > n3 = Gravity Override State Function DEPRECATED(
    CharacterPartAnimInstance=Self-Reference()
    GameTimeInSeconds=0
    GravityOverrideParams=Get Gravity Override Params()
    RBAN Params=Get RBAN Params()
   )
   > Set Gravity Override(Gravity Override=@n3)
   > n4 = Sequence()
   n5 = float * float(A=GetCurveValue(CurveName=FrontWheelSpin), B=Get Delta Time X()) {is=title K2Node_PromotableOperator}
   n6 = Get Wheel Rotation Last Tick_Front() -> (Wheel Rotation Last Tick_Front:double)
   > n7 = Set WheelRotation_Front(WheelRotation_Front:double="% (Float)"!(A:double=float + float(A=@n5, B=@n6), B:double=360) -> (ReturnValue:double)) -> (Output_Get:double)
   > Set WheelSpin_Front(WheelSpin_Front:Rotator=MakeRotator(Roll=@n7)) -> (Output_Get:Rotator)
   > Set Wheel Rotation Last Tick_Front(Wheel Rotation Last Tick_Front:double=Get WheelRotation_Front() -> (WheelRotation_Front:double)) -> (Output_Get:double)
   n8 = float * float(A=GetCurveValue(CurveName=RearWheelSpin), B=Get Delta Time X())
   n9 = Get Wheel Rotation Last Tick_Rear() -> (Wheel Rotation Last Tick_Rear:double)
   n10 = Set WheelRotation_Rear(execute=@n4.then_1, WheelRotation_Rear:double="% (Float)"(A=float + float(A=@n8, B=@n9), B=360)) -> (Output_Get:double)
   > Set WheelSpin_Rear(WheelSpin_Rear:Rotator=MakeRotator(Roll=@n10)) -> (Output_Get:Rotator)
   > Set Wheel Rotation Last Tick_Rear(Wheel Rotation Last Tick_Rear:double=Get WheelRotation_Rear() -> (WheelRotation_Rear:double)) -> (Output_Get:double)
   n11 = AnimNotify_FE_ReactiveReset() {is=title K2Node_Event}
   Set Active Kill(execute=[@n11.then, @n14.then], Active Kill:bool=false) -> (Output_Get:bool)
   > Branch(Condition=Get bShouldDisplayFrontendReactivity() -> (bShouldDisplayFrontendReactivity:bool))
   > Delay(Duration=10)
   > Set Active Kill(Active Kill=true)
   n12 = Event OnCharacterCustomizationLoaded()
   > Set Custom Base Mesh(Custom Base Mesh:Skeletal Mesh Component Object Reference=@n12.CustomBaseMesh) -> (Output_Get:Skeletal Mesh Component Object Reference)
   n13 = Event BlueprintProcessAnimInputEvent()
   > Branch(Condition=IsNamed(self=@n13.AnimInputEvent, InEventName=aie.EnemyPlayerEliminated))
   > Set Active Kill(Active Kill=true)
   n14 = Event BlueprintPostEvaluateAnimation() {is=title K2Node_Event}
   TryGetPawnOwner()
   GetCurveValue(CurveName=None)
  ]}
]
variables [
 Active Kill {type=boolean}
 Custom Base Mesh {type=SkeletalMeshComponent}
 Delta Time X {type=number}
 Gravity Override {type=Vector}
 Gravity Override Params {type=GravityOverrideParamsStruct}
 RBAN Params {
  type=RBANParamsStruct
  default {
   dBNO {simSpaceSettings {
     maxAngularAcceleration=1000
     maxAngularVelocity=3
     maxLinearAcceleration=300
     maxLinearVelocity=700
     velocityScaleZ=0
     externalLinearDragV=<0 0.5 0>
    }}
   dBNO_Carried {simSpaceSettings {
     maxAngularAcceleration=1000
     maxAngularVelocity=3
     maxLinearAcceleration=300
     maxLinearVelocity=700
     velocityScaleZ=0
     externalLinearDragV=<0 0.5 0>
    }}
   emoteOrMelee {simSpaceSettings {
     maxAngularAcceleration=1000
     maxAngularVelocity=3
     maxLinearAcceleration=300
     maxLinearVelocity=700
     velocityScaleZ=0
     externalLinearDragV=<0 0.5 0>
    }}
   front_End {simSpaceSettings {
     maxAngularAcceleration=1000
     maxAngularVelocity=3
     maxLinearAcceleration=300
     maxLinearVelocity=700
     velocityScaleZ=0
     externalLinearDragV=<0 0.5 0>
    }}
   onGround_Crouch_Moving {simSpaceSettings {
     maxAngularAcceleration=1000
     maxAngularVelocity=3
     maxLinearAcceleration=300
     maxLinearVelocity=700
     velocityScaleZ=0
     externalLinearDragV=<0 0.5 0>
    }}
   onGround_Falling {simSpaceSettings {
     maxAngularAcceleration=1000
     maxAngularVelocity=3
     maxLinearAcceleration=300
     maxLinearVelocity=700
     velocityScaleZ=0
     externalLinearDragV=<0 0.5 0>
    }}
   onGround_Moving {simSpaceSettings {
     maxAngularAcceleration=1000
     maxAngularVelocity=3
     maxLinearAcceleration=300
     maxLinearVelocity=700
     velocityScaleZ=0
     externalLinearDragV=<0 0.5 0>
    }}
   onGround_Standing {simSpaceSettings {
     maxAngularAcceleration=1000
     maxAngularVelocity=3
     maxLinearAcceleration=300
     maxLinearVelocity=700
     velocityScaleZ=0
     externalLinearDragV=<0 0.5 0>
    }}
   rail_Grinding {simSpaceSettings {
     maxAngularAcceleration=1000
     maxAngularVelocity=3
     maxLinearAcceleration=300
     maxLinearVelocity=700
     velocityScaleZ=0
     worldAlpha=0.1
     externalLinearDragV=<0 0.5 0>
    }}
   riding_Moving {simSpaceSettings {
     maxAngularAcceleration=1000
     maxAngularVelocity=3
     maxLinearAcceleration=300
     maxLinearVelocity=700
     velocityScaleZ=0
     externalLinearDragV=<0 0.5 0>
    }}
   riding_Moving_Targeting {simSpaceSettings {
     maxAngularAcceleration=1000
     maxAngularVelocity=3
     maxLinearAcceleration=300
     maxLinearVelocity=700
     velocityScaleZ=0
     externalLinearDragV=<0 0.5 0>
    }}
   riding_Standing {simSpaceSettings {
     maxAngularAcceleration=1000
     maxAngularVelocity=3
     maxLinearAcceleration=300
     maxLinearVelocity=700
     velocityScaleZ=0
     externalLinearDragV=<0 0.5 0>
    }}
   skydive_DivingDown {simSpaceSettings {
     maxAngularAcceleration=1000
     maxAngularVelocity=3
     maxLinearAcceleration=300
     maxLinearVelocity=700
     velocityScaleZ=0
     externalLinearDragV=<0 0.5 0>
    }}
   skydive_FloatingDown {
    componentLinearAccScale=<0 0.5 0>
    componentLinearVelScale=<0 0.5 0>
    simSpaceSettings {
     maxAngularAcceleration=1000
     maxAngularVelocity=3
     maxLinearAcceleration=300
     maxLinearVelocity=700
     velocityScaleZ=0
     externalLinearDragV=<0 0.5 0>
    }
   }
   skydive_GoingUp {simSpaceSettings {
     maxAngularAcceleration=1000
     maxAngularVelocity=3
     maxLinearAcceleration=300
     maxLinearVelocity=700
     velocityScaleZ=0
     externalLinearDragV=<0 0.5 0>
    }}
   skydive_Parachuting {
    componentLinearAccScale=<0 0.1 0.3>
    componentLinearVelScale=<0 0.1 0.3>
    simSpaceSettings {
     maxAngularAcceleration=1000
     maxAngularVelocity=3
     maxLinearAcceleration=300
     maxLinearVelocity=700
     velocityScaleZ=0
     externalLinearDragV=<0 0.5 0>
    }
   }
   skydive_Surfing {simSpaceSettings {
     maxAngularAcceleration=1000
     maxAngularVelocity=3
     maxLinearAcceleration=300
     maxLinearVelocity=700
     velocityScaleZ=0
     externalLinearDragV=<0 0.5 0>
    }}
   swimming {simSpaceSettings {
     maxAngularAcceleration=1000
     maxAngularVelocity=3
     maxLinearAcceleration=300
     maxLinearVelocity=700
     velocityScaleZ=0
     externalLinearDragV=<0 0.5 0>
    }}
   tactical_Sprint {simSpaceSettings {
     maxAngularAcceleration=1000
     maxAngularVelocity=3
     maxLinearAcceleration=300
     maxLinearVelocity=700
     velocityScaleZ=0
     externalLinearDragV=<0 0.5 0>
    }}
   ziplining {simSpaceSettings {
     maxAngularAcceleration=1000
     maxAngularVelocity=3
     maxLinearAcceleration=300
     maxLinearVelocity=700
     velocityScaleZ=0
     externalLinearDragV=<0 0.5 0>
    }}
  }
 }
 TActor {type=Transform, default {location=<0 0 0>, rotation=<rot 0 0 0>, scale=<1 1 1>}}
 TPelvis {type=Transform, default {location=<0 0 0>, rotation=<rot 0 0 0>, scale=<1 1 1>}}
 TSpine03 {type=Transform, default {location=<0 0 0>, rotation=<rot 0 0 0>, scale=<1 1 1>}}
 TSpine_02 {type=Transform, default {location=<0 0 0>, rotation=<rot 0 0 0>, scale=<1 1 1>}}
 TSpine_05 {type=Transform, default {location=<0 0 0>, rotation=<rot 0 0 0>, scale=<1 1 1>}}
 Wheel Rotation Last Tick_Front {type=number}
 Wheel Rotation Last Tick_Rear {type=number}
 WheelRotation_Front {type=number}
 WheelRotation_Rear {type=number}
 WheelSpin_Front {type=Rotator}
 WheelSpin_Rear {type=Rotator}
]
